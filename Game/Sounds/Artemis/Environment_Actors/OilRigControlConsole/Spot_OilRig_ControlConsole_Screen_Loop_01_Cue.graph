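# Looping ambient sound cue for an oil rig control console screen, classed as a gadget sound.
# It mixes two reused computer loops, a transmitter hum and a quieter glitch layer, then applies slight random pitch modulation.
# A quality-level root feeds that same chain to the first two slots and leaves the other two empty.
# A dedicated attenuation asset spatializes it, and a concurrency override caps simultaneous instances.
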SoundCue /Game/Sounds/Artemis/Environment_Actors/OilRigControlConsole/Spot_OilRig_ControlConsole_Screen_Loop_01_Cue
root=QualityLevel_0
nodes [
 {id=Mixer_0, children [WavePlayer_0,WavePlayer_2], properties {inputVolume [1,0.5]}}
 {id=Modulator_0, children [Mixer_0], properties {pitchMax=1.02, pitchMin=0.98, volumeMax=1, volumeMin=1}}
 {id=QualityLevel_0, children [Modulator_0,Modulator_0,null,null]}
 {
  id=WavePlayer_0
  properties {bLooping=true, soundWaveAssetPtr=/Game/Athena/Sounds/Props/MountainBaseComputer/ComputerTransimtter_Running_Loop_01}
 }
 {id=WavePlayer_2, properties {bLooping=true, soundWaveAssetPtr=/Game/Sounds/Apollo/Biomes/SwampShack/CavernShack_Computer_Glitch_Loop_01}}
]
properties {
 attenuationSettings=&/Spot_Console_Screen_Loop_Att
 bOverrideConcurrency=true
 maxDistance=650
 soundClassObject=&Fort_Gadgets
 volumeMultiplier=0.5
 attenuationOverrides {
  bEnableFocusInterpolation=true
  falloffDistance=600
  nonFocusVolumeAttenuation=0.3
  occlusionVolumeAttenuation=0.3
  stereoSpread=0
  attenuationShapeExtents {x=200}
 }
 concurrencyOverrides {maxCount=2}
}
