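# Sound cue for a looping screen sound on the oil rig control console, routed to the Creative gadgets sound class at reduced volume.
# A single looping wave player is the root, reusing a lock device screen loop from the Creative toys sounds.
# Spatialization references a console screen loop attenuation asset.
# A concurrency override caps it at two simultaneous instances, stopping the quietest.
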SoundCue /Game/Sounds/Artemis/Environment_Actors/OilRigControlConsole/Spot_OilRig_ControlConsole_Screen_Loop_02_Cue
root=WavePlayer_0
nodes [{id=WavePlayer_0, properties {bLooping=true, soundWaveAssetPtr=/Game/Sounds/Creative/Toys/Locks/LockDevice_Screen_Loop_02}}]
properties {
 attenuationSettings=&/Spot_Console_Screen_Loop_Att
 bOverrideConcurrency=true
 maxDistance=650
 soundClassObject=&Fort_Gadgets_Creative
 volumeMultiplier=0.5
 attenuationOverrides {
  bEnableFocusInterpolation=true
  bEnableListenerFocus=true
  falloffDistance=400
  nonFocusVolumeAttenuation=0.5
  stereoSpread=500
  attenuationShapeExtents {x=0}
 }
 concurrencyOverrides {maxCount=2, resolutionRule=StopQuietest}
}
